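# Physics asset for the face accessory mesh of the Pizza Pit Mascot skin, covering its eye and leaf bones.
# Boxes wrap the eyes and leaf segments, while small spheres serve as anchor bodies on the base and center bones.
# Constraints chain leaf segments to their anchors with narrow swing and twist limits, tighter on the back-right chain.
# Both eye constraints lock swing and twist, holding each eye's rotation fixed to its base.
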PhysicsAsset /BRCosmetics/Characters/Player/Female/Medium/Bodies/F_Med_Soldier_01/Skins/PizzaPitMascot/Meshes/F_MED_Pizza_Pit_Mascot_FaceAcc_Physics
bodies {
 dyn_eye_l [{shapeType=Box, extentX=2.2296, extentY=0.5983, extentZ=2.667, center=<-0.0031 -0.0312 -0.0198>, rotation=<rot 22.1493 -90 0.8461>}]
 dyn_eye_l_base [{shapeType=Sphere, radius=0.5}]
 dyn_eye_r [{shapeType=Box, extentX=1.8308, extentY=0.2069, extentZ=2.6392, center=<-0.0049 0.0434 -0.0387>, rotation=<rot -22.809 -90 0.1502>}]
 dyn_eye_r_base [{shapeType=Sphere, radius=0.5}]
 dyn_leaf_back_l [{shapeType=Box, extentX=3.6361, extentY=3.9471, extentZ=1.9923, center=<2.2363 0 0.0003>}]
 dyn_leaf_back_l_01 [{shapeType=Box, extentX=3.7442, extentY=4.0875, extentZ=1.318, center=<1.9847 0 -0.0001>}]
 dyn_leaf_back_l_base [{shapeType=Sphere, radius=0.5}]
 dyn_leaf_back_r [{shapeType=Box, extentX=4.425, extentY=5.2999, extentZ=2.186, center=<2.803 0 0>}]
 dyn_leaf_back_r_01 [{shapeType=Box, extentX=6.676, extentY=4.9705, extentZ=1.8328, center=<3.1675 0 0>}]
 dyn_leaf_back_r_base [{shapeType=Sphere, radius=0.5}]
 dyn_leaf_center [{shapeType=Sphere, radius=0.875}]
 dyn_leaf_front_l [{shapeType=Box, extentX=4.9768, extentY=5.8571, extentZ=1.5, center=<2.5749 0 0>}]
 dyn_leaf_front_l_01 [{shapeType=Box, extentX=5.4864, extentY=4.1322, extentZ=1.2697, center=<2.5023 0 0>}]
 dyn_leaf_front_l_base [{shapeType=Sphere, radius=0.5}]
 dyn_leaf_front_r [{shapeType=Box, extentX=6.3539, extentY=5.6748, extentZ=1.25, center=<2.6174 0 0.0001>}]
}
constraints [
 {bone1Name=dyn_leaf_back_r, bone2Name=dyn_leaf_back_r_base, swing1LimitDegrees=20, swing2LimitDegrees=10, twistLimitDegrees=3}
 {bone1Name=dyn_leaf_back_r_01, bone2Name=dyn_leaf_back_r, swing1LimitDegrees=10, swing2LimitDegrees=5, twistLimitDegrees=3}
 {bone1Name=dyn_leaf_front_l_01, bone2Name=dyn_leaf_front_l, swing1LimitDegrees=20, swing2LimitDegrees=20, twistLimitDegrees=3}
 {bone1Name=dyn_leaf_front_r, bone2Name=dyn_leaf_center, swing1LimitDegrees=20, swing2LimitDegrees=20, twistLimitDegrees=3}
 {bone1Name=dyn_leaf_front_l, bone2Name=dyn_leaf_front_l_base, swing1LimitDegrees=20, swing2LimitDegrees=20, twistLimitDegrees=3}
 {bone1Name=dyn_leaf_back_l, bone2Name=dyn_leaf_back_l_base, swing1LimitDegrees=20, swing2LimitDegrees=20, twistLimitDegrees=3}
 {bone1Name=dyn_leaf_back_l_01, bone2Name=dyn_leaf_back_l, swing1LimitDegrees=20, swing2LimitDegrees=20, twistLimitDegrees=3}
 {
  bone1Name=dyn_eye_r
  bone2Name=dyn_eye_r_base
  swing1LimitDegrees=30
  swing1Motion=Locked
  swing2LimitDegrees=30
  swing2Motion=Locked
  twistLimitDegrees=45
  twistMotion=Locked
 }
 {
  bone1Name=dyn_eye_l
  bone2Name=dyn_eye_l_base
  swing1LimitDegrees=30
  swing1Motion=Locked
  swing2LimitDegrees=30
  swing2Motion=Locked
  twistLimitDegrees=45
  twistMotion=Locked
 }
]
